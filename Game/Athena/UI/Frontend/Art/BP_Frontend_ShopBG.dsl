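; Frontend item shop backdrop actor that starts hidden and dresses the shop background for seasonal or event environments.
; A string switch on the CMS preset name swaps plane materials, visibility or snow effects, while SetDynamicBG feeds a CMS texture into a dynamic material.
; For donut presets, shop enter and hide events toggle a tagged dome obstruction and material opacity.
; FullDonut also drives the CMSLobbyDirector's sequence state based on whether it was seen.
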(event EventBeginPlay
  ;[0 0 0] Hide everything as default right now
  (Rendering|SetActorHiddenInGame true)
  (Utilities|FlowControl|Sequence
    (:then_0
      ;[0 0 0] get the cms director for donut sequence driving
      (bind _outactors (Actor|GetAllActorsOfClass "/Game/Environments/FrontEnd/BackPlates/CMSLobbyDirector.CMSLobbyDirector_C"))
      (if (> (Utilities|Array|Length _outactors) 0)
        (Variables|Default|SetCMSDirector (Utilities|Array|Get(acopy) _outactors))))
    (:then_1
      (bind _outactors_2 (Actor|GetAllActorswithTag "DonutObstruction"))
      (if (> (Utilities|Array|Length _outactors_2) 0)
        (Variables|Default|SetDomeObstruction (Utilities|Array|Get(acopy) _outactors_2))))))

(event Custom|SetDynamicEnvironmentEvent
  (params (InPresetEnvName :type String))
  (bind _standardbackgroundplane (Variables|Default|GetStandardBackgroundPlane))
  (bind _standardbackgroundplane_2 (Variables|Default|GetStandardBackgroundPlane))
  (bind _standardbackgroundplane_3 (Variables|Default|GetStandardBackgroundPlane))
  (bind _standardbackgroundplane1 (Variables|Default|GetStandardBackgroundPlane1))
  (bind _cmsdirector (Variables|Default|GetCMSDirector))
  (bind _standardbackgroundplane_4 (Variables|Default|GetStandardBackgroundPlane))
  (bind _standardbackgroundplane1_2 (Variables|Default|GetStandardBackgroundPlane1))
  (bind _cmsdirector_2 (Variables|Default|GetCMSDirector))
  (bind _standardbackgroundplane_5 (Variables|Default|GetStandardBackgroundPlane))
  ;[0 0 0] todo: things here switching on cms stage name
  (Variables|Default|SetbHasEvent false)
  (Variables|Default|SetBPEnvName InPresetEnvName)
  (Utilities|IsValid _cmsdirector
    (:Is%20Valid
      (Class|DynamicBackgroundDirector|SetIsinDonutSequence _cmsdirector false))
    (:Is%20Not%20Valid
      _))
  (switch string InPresetEnvName
    (:Default
      _)
    (:blackmonday
      ;;[0 0 0] black monday
      (Rendering|SetVisibility (Variables|Default|GetVeryLargeBackgroundPlane) true true)
      (Rendering|SetVisibility (Variables|Default|GetBackgroundTexturePlane) false false))
    (:fortnitemares
      ;;[0 0 0] Fortnitemares
      (Rendering|Material|SetMaterial _standardbackgroundplane_2 0 "/Game/Environments/FrontEnd/BackPlates/Materials/M_FrontendBackplate_Master_FNMares_Shop.M_FrontendBackplate_Master_FNMares_Shop")
      (Rendering|SetVisibility _standardbackgroundplane_2 true false))
    (:GalileoLite
      ;;[0 0 0] Galileo and Galielo lite both get the same background
      (Rendering|Material|SetMaterial _standardbackgroundplane_3 0 "/Game/Environments/FrontEnd/BackPlates/Materials/M_FrontendBackplate_Galileo_ItemShop.M_FrontendBackplate_Galileo_ItemShop")
      (Rendering|SetVisibility _standardbackgroundplane_3 true false))
    (:Galileo
      _)
    (:winter19
      ;;[0 0 0] Winterfest 2019
      (Rendering|Material|SetMaterial _standardbackgroundplane_5 0 "/Game/Environments/FrontEnd/BackPlates/Materials/M_FrontendBackplate_Master1_Winter_Shop.M_FrontendBackplate_Master1_Winter_Shop")
      (Rendering|SetVisibility _standardbackgroundplane_5 true false))
    (:FullDonut
      ;;[0 0 0] donut animation
      (bind _returnvalue (Rendering|Material|CreateDynamicMaterialInstance "/Game/Environments/FrontEnd/BackPlates/Materials/M_DonutStore_Inst.M_DonutStore_Inst" "None" "None" :!class /Script/Engine.KismetMaterialLibrary))
      (bind _output_get (Variables|Default|SetMID _returnvalue))
      (Rendering|Material|SetMaterial _standardbackgroundplane1_2 0 _output_get)
      (Rendering|SetVisibility _standardbackgroundplane1_2 true false)
      ;[0 0 0] transitioning when we are already active and open
      (if (Variables|Default|GetShopEntered)
        (Utilities|IsValid _cmsdirector_2
          (:Is%20Valid
            (bind _returnvalue_2 (Class|DynamicBackgroundDirector|GetHasDonutShopSequenceBeenSeen _cmsdirector_2))
            (if (not _returnvalue_2)
              (Class|DynamicBackgroundDirector|TryForceSetupDonutStore _cmsdirector_2))
            (CallFunction|HandleItemShopEntered)))))
    (:DonutBG
      (bind _returnvalue_3 (Rendering|Material|CreateDynamicMaterialInstance "/Game/Environments/FrontEnd/BackPlates/Materials/M_DonutStore_Inst.M_DonutStore_Inst" "None" "None" :!class /Script/Engine.KismetMaterialLibrary))
      (bind _output_get_2 (Variables|Default|SetMID _returnvalue_3))
      (Rendering|Material|SetMaterial _standardbackgroundplane1 0 _output_get_2)
      (Rendering|SetVisibility _standardbackgroundplane1 true false))
    (:Cyclone
      (Rendering|Material|SetMaterial _standardbackgroundplane 0 "/Game/Environments/FrontEnd/BackPlates/Materials/M_FrontendBackplate_CycloneShop.M_FrontendBackplate_CycloneShop")
      (Rendering|SetVisibility _standardbackgroundplane true false))
    (:CandyApple
      _)
    (:XmasStore2020
      ;[0 0 0] XMas Store2020
      (Rendering|Material|SetMaterial _standardbackgroundplane_4 0 "/Game/Environments/FrontEnd/BackPlates/Materials/M_SimpleStoreBackgorund.M_SimpleStoreBackgorund")
      (Rendering|SetVisibility _standardbackgroundplane_4 true false)
      (Effects|SpawnEmitteratLocation "/Game/Effects/Fort_Effects/World/Enviroment/Snow/P_Snow_Ambient_Lobby_warmup2020.P_Snow_Ambient_Lobby_warmup2020" "-413.943817,30695.449219,-684.322266")))
  (Rendering|SetVisibility (Variables|Default|GetVeryLargeBackgroundPlane) false true))

(event Custom|EnableDynamicShopBG
  ;[0 0 0] ke* EnableDynamicShopBG to turn on vis - DEVELOPMENT ONLY
  (Development|SetHiddeninGame (Variables|Default|GetDefaultSceneRoot) false true))

(event Custom|SetDynamicBG
  (params (InTexture :type Texture%20Object%20Reference))
  (bind _backgroundtextureplane (Variables|Default|GetBackgroundTexturePlane))
  ;[0 0 0] apply the CMS texture to background plane
  (Utilities|IsValid InTexture
    (:Is%20Valid
      (bind _returnvalue (Rendering|Material|CreateDynamicMaterialInstance _backgroundtextureplane 0 "/Game/Athena/UI/Frontend/Art/M_Frontend_DynamicBG.M_Frontend_DynamicBG" "None" :!class /Script/Engine.PrimitiveComponent))
      (Rendering|Material|SetTextureParameterValue _returnvalue "InTextureObject" InTexture))
    (:Is%20Not%20Valid
      (Rendering|SetVisibility _backgroundtextureplane false false))))

(event Custom|DebugShopBG)

(event Custom|HandleStoreEntered
  ;[0 0 0] store tab entered
  (switch string (Variables|Default|GetBPEnvName)
    (:FullDonut
      _)
    (:DonutBG
      _))
  ;;[0 0 0] turn off Donut for the store
  (Rendering|SetActorHiddenInGame true)
  (CallFunction|HandleRequestHide))

(event Custom|ResetDonut
  (bind _cmsdirector (Variables|Default|GetCMSDirector))
  ;[0 0 0] ke* ResetDonut : debug reset marked as seen flag
  (Utilities|IsValid _cmsdirector
    (:Is%20Valid
      (Class|DynamicBackgroundDirector|ResetDonutShopSeen _cmsdirector))))

(event Custom|HandleItemShopEntered
  (bind _cmsdirector (Variables|Default|GetCMSDirector))
  (bind _domeobstruction (Variables|Default|GetDomeObstruction))
  (bind _domeobstruction_2 (Variables|Default|GetDomeObstruction))
  (bind _domeobstruction_3 (Variables|Default|GetDomeObstruction))
  ;[0 0 0] Item shop tab entered
  (Variables|Default|SetShopEntered true)
  (switch string (Variables|Default|GetBPEnvName)
    (:FullDonut
      (Utilities|IsValid _cmsdirector
        (:Is%20Valid
          (bind _returnvalue (Class|DynamicBackgroundDirector|GetHasDonutShopSequenceBeenSeen _cmsdirector))
          (if (not _returnvalue)
            (Utilities|IsValid _domeobstruction
              (:Is%20Valid
                (Rendering|SetActorHiddenInGame _domeobstruction false))
              (:Is%20Not%20Valid
                _))
            (Rendering|Material|SetScalarParameterValue (Variables|Default|GetMID) "Opacity" 0.0 :!class /Script/Engine.MaterialInstanceDynamic)
            (Class|DynamicBackgroundDirector|SetIsinDonutSequence (Variables|Default|GetCMSDirector) true)
            (else
              (Utilities|IsValid _domeobstruction_3
                (:Is%20Valid
                  (Rendering|SetActorHiddenInGame _domeobstruction_3 true)
                  (Rendering|Material|SetScalarParameterValue (Variables|Default|GetMID) "Opacity" 1.0 :!class /Script/Engine.MaterialInstanceDynamic)
                  (Class|DynamicBackgroundDirector|SetIsinDonutSequence (Variables|Default|GetCMSDirector) false))))))))
    (:DonutBG
      (Utilities|IsValid _domeobstruction_2
        (:Is%20Valid
          (Rendering|SetActorHiddenInGame _domeobstruction_2 true)
          (Rendering|Material|SetScalarParameterValue (Variables|Default|GetMID) "Opacity" 1.0 :!class /Script/Engine.MaterialInstanceDynamic))))))

(event Custom|HandleRequestHide
  (bind _domeobstruction (Variables|Default|GetDomeObstruction))
  ;[0 0 0] handle any special things we need to hide when closed
  (Variables|Default|SetShopEntered false)
  (Utilities|Time|ClearandInvalidateTimerbyHandle (Variables|Default|GetDonutSeqTimerHandle))
  (Utilities|IsValid _domeobstruction
    (:Is%20Valid
      (Rendering|SetActorHiddenInGame _domeobstruction false))))

(fn ConstructionScript)
