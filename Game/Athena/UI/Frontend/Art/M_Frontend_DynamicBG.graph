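# M_Frontend_DynamicBG is an unlit, two-sided material, named as a dynamic frontend background, with no material outputs connected in the graph.
# Screen-aligned UVs drive two panning cloud samples, multiplied into moving noise.
# That noise is scaled by Opacity and a blank-texture green channel minus itself, which evaluates to zero.
# Separate unconnected branches tint that channel with CircleColor and sample the InTextureObject parameter.
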Material /Game/Athena/UI/Frontend/Art/M_Frontend_DynamicBG
graph {
 properties {pixelDepthOffsetMode=PDOM_Legacy, shadingModel=MSM_Unlit, twoSided=true, usedShadingModels=Unlit}
 nodes [
  Constant3Vector()
  Constant3Vector()
  n1 = Multiply(A=MaterialFunctionCall(Texture Dimensions=Constant2Vector() {g=256, r=256}) {materialFunction=&ScreenAlignedUVs}."X 100%, Y 100%") {constB=0.85}
  n2 = Panner(Coordinate=@n1) {speedX=0.07, speedY=0.02}
  n3 = Multiply(
   A=Multiply(A=TextureSample(UVs=@n2) {texture=/Game/Packages/Fortnite_UniqueMaterials_1/PostProcess/T_TilingClouds_02}.R)
   B=TextureSample(UVs=Panner(Coordinate=@n1) {speedX=0.05}) {texture=&T_TilingClouds_02}.R
  )
  n4 = TextureSample() {samplerSource=SSM_Clamp_WorldGroupSettings, texture=/Game/Global/Textures/Default/Blanks/T_Blank_White_ZeroAlpha_sRGBColor}
  Multiply(
   A=Clamp(None=Add(A=@n3) {constB=0.05})
   B=Multiply(A=Subtract(A=@n4.G, B=@n4.G), B=ScalarParameter() {defaultValue=0.2714, parameterName=Opacity})
  )
  Multiply(A=Multiply(A=VectorParameter() {parameterName=CircleColor, defaultValue=<rgba 0.3438 0.7758 0.7216 1>}.RGB, B=@n4.G)) {constB=2.7075}
  TextureSample(Tex=TextureObjectParameter() {parameterName=InTextureObject}) {samplerSource=SSM_Clamp_WorldGroupSettings, texture=&T_Blank_White_ZeroAlpha_sRGBColor}
 ]
}
