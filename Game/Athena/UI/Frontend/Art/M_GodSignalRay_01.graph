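# Additive, unlit, two-sided material for a light ray in the frontend, judging by its name and path.
# A Fresnel term blends two tint colors for a dimmed glow and, multiplied by a V gradient, shapes the ray mask.
# Two panning channels of a packed noise texture add moving breakup, then depth fade and a smoothstepped end gradient soften the mask.
# A static switch can add near-camera fading.
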Material /Game/Athena/UI/Frontend/Art/M_GodSignalRay_01
graph {
 properties {blendMode=BLEND_Additive, pixelDepthOffsetMode=PDOM_Legacy, shadingModel=MSM_Unlit, twoSided=true, usedShadingModels=Unlit}
 nodes [
  n1 = MaterialFunctionCall(
   Brightness=ScalarParameter() {defaultValue=1.1, parameterName=Fresnel Brightness}
   Exponent=ScalarParameter() {defaultValue=3, parameterName=Fresnel Exponent}
  ) {materialFunction=&MF_Fresnel_FX}
  Multiply(A=LinearInterpolate(
    A=VectorParameter() {parameterName=Color A, defaultValue=<rgba 0.0103 0.006 0.0343 1>}.RGB
    B=VectorParameter() {parameterName=Color B, defaultValue=<rgba 0.4 0.2904 0.15 0>}.RGB
    Alpha=@n1.Fresnel
   )) {constB=0.35}
  n2 = ComponentMask(None=TextureCoordinate() {uTiling=2, vTiling=0.127}) {g=true}
  n3 = Power(Base=Subtract(A=OneMinus(None=@n2)) {constB=0.87}) {constExponent=0.8}
  Multiply(A=@n3, B=ScalarParameter() {defaultValue=9.5946, parameterName=EndBlendMult}) {constB=2}
  n4 = Multiply(A=@n1.Fresnel, B=ComponentMask(None=TextureCoordinate()) {g=true})
  n5 = Panner(Coordinate=TextureCoordinate() {uTiling=2, vTiling=0.127}) {bFractionalPart=true, speedX=0.055, speedY=0.0355}
  n6 = Panner(Coordinate=TextureCoordinate() {vTiling=0.139}) {bFractionalPart=true, speedX=0.0538, speedY=0.0505}
  n7 = Add(
   A=Add(
    A=TextureSample(UVs=@n5) {texture=&T_NoisePack_01}.R
    B=TextureSample(UVs=@n6) {texture=&T_NoisePack_01}.G
   )
   B=ScalarParameter() {defaultValue=0.2, parameterName=Boost}
  )
  n8 = ScalarParameter() {defaultValue=0.01, parameterName=smoothCutoff}
  n9 = Multiply(
   A=DepthFade(Opacity=Multiply(A=@n4, B=@n7), FadeDistance=ScalarParameter() {defaultValue=1000, parameterName=DepthFade})
   B=MaterialFunctionCall(Alpha=@n3, Min=@n8, Max=Add(A=@n3, B=@n8) {constB=0.001}) {materialFunction=&SmoothStep}.Result
  )
  n10 = ScalarParameter() {parameterName=NearCameraFadeDistance}
  StaticSwitchParameter(
   True=Multiply(A=@n9, B=MaterialFunctionCall(Distance=@n10) {materialFunction=&Near_Camera_Fade}.Result)
   False=@n9
  ) {parameterName=UseNearCameraFade}
 ]
}
