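# Translucent UI material for a "14 Days" frontend background, showing a gradient ribbon over a gradient backdrop with panning snowflakes.
# Rotated UV4 edges and notches form the ribbon mask that blends a linear ribbon gradient with a radial or linear background.
# R, G and B texture channels form three screen-aligned flake layers, each thinned to alternate tiles, that tint the colors.
# Static switches control flake placement, ribbon masking, gradient type and optional flake-driven opacity.
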Material /Game/Athena/UI/Frontend/Art/M_UI_14DaysBackGround
graph {
 properties {
  bUsedWithStaticMesh=false
  blendMode=BLEND_Translucent
  materialDomain=MD_UI
  pixelDepthOffsetMode=PDOM_Legacy
  usedShadingModels=Default Lit
 }
 nodes [
  n1 = TextureCoordinate() {coordinateIndex=4}
  n2 = ScalarParameter() {group=GradientSettings, parameterName=RibbonGradientOffset}
  n3 = Rotator(
   Coordinate=Multiply(A=Add(A=@n1, B=@n2) {constB=0.5}, B=ScalarParameter() {defaultValue=1, group=GradientSettings, parameterName=RibbonGradientScale}) {constB=0.5}
   Time=Multiply(A=ScalarParameter() {defaultValue=40, group=GradientSettings, parameterName=RibbonGradientRotation}) {constB=0.0028}
  ) {centerY=0, speed=6.2832}
  n4 = LinearInterpolate(
   A=VectorParameter() {group=Colors, parameterName=Ribbons_RGB1, defaultValue=<rgba 0.0382 0 0.3372 1>}.RGB
   B=VectorParameter() {group=Colors, parameterName=Ribbons_RGB2, defaultValue=<rgba 0.013 0 0.0908 1>}.RGB
   Alpha=Clamp(None=ComponentMask(None=@n3) {g=true})
  )
  n5 = VectorParameter() {group=Colors, parameterName=SnowflakeRGBMultiplier, defaultValue=<rgba 0.375 0.375 0.375 1>}
  n6 = Multiply(A=@n4, B=@n5.RGB) {constB=0.25}
  n7 = MaterialFunctionCall() {materialFunction=&ScreenAlignedUVs}
  n8 = AppendVector(
   A=ScalarParameter() {defaultValue=0.1, group=Snowflake_R, parameterName=Flake_R_OffsetX}
   B=ScalarParameter() {defaultValue=0.1, group=Snowflake_R, parameterName=Flake_R_OffsetY}
  )
  n9 = Panner(
   Coordinate=Rotator(
    Coordinate=Add(A=@n7."X 100%, Y 100%", B=@n8) {constB=0.255}
    Time=ScalarParameter() {defaultValue=2.8047, group=Snowflake_R, parameterName=Flake_R_Rotation}
   )
   Speed=AppendVector(
    A=ScalarParameter() {defaultValue=-0.05, group=Snowflake_R, parameterName=Flake_R_SpeedX}
    B=ScalarParameter() {defaultValue=-0.1, group=Snowflake_R, parameterName=Flake_R_SpeedY}
   )
  ) {speedX=0.25}
  n10 = ScalarParameter() {defaultValue=8, group=Snowflake_R, parameterName=Flake_R_TileU}
  n11 = ScalarParameter() {defaultValue=6, group=Snowflake_R, parameterName=Flake_R_TileV}
  n12 = AppendVector(A=@n10, B=@n11)
  n13 = TextureSampleParameter2D(UVs=Frac(None=Multiply(A=@n9, B=@n12))) {group=SnowflakeSettings, parameterName=PanningTexture, texture=/Game/UI/Foundation/Textures/Pattern/T_14Days_Snowflakes}
  n14 = ScalarParameter() {defaultValue=32, group=SnowflakeSettings, parameterName=SnowflakeSharpness}
  n15 = AppendVector(A=Multiply(A=@n10) {constB=0.5}, B=Multiply(A=@n11) {constB=0.5})
  n16 = Add(A=Frac(None=Multiply(A=@n9, B=@n15))) {constB=-0.5}
  n17 = Clamp(None=Multiply(A=ComponentMask(None=@n16) {g=true}) {constB=256})
  n18 = Clamp(None=Multiply(A=ComponentMask(None=@n16) {r=true}) {constB=256})
  n19 = LinearInterpolate(B=Clamp(None=Multiply(A=@n13.R, B=@n14) {constB=32}), Alpha=OneMinus(None=LinearInterpolate(A=@n17, Alpha=@n18)))
  n20 = AppendVector(
   A=ScalarParameter() {defaultValue=0.25, group=Snowflake_G, parameterName=Flake_G_OffsetX}
   B=ScalarParameter() {defaultValue=0.1, group=Snowflake_G, parameterName=Flake_G_OffsetY}
  )
  n21 = Panner(
   Coordinate=Rotator(
    Coordinate=Add(A=@n7."X 100%, Y 100%", B=@n20) {constB=0.255}
    Time=ScalarParameter() {defaultValue=0.1, group=Snowflake_G, parameterName=Flake_G_Rotation}
   )
   Speed=AppendVector(
    A=ScalarParameter() {defaultValue=-0.05, group=Snowflake_G, parameterName=Flake_G_SpeedX}
    B=ScalarParameter() {defaultValue=-0.1, group=Snowflake_G, parameterName=Flake_G_SpeedY}
   )
  ) {speedX=0.25}
  n22 = ScalarParameter() {defaultValue=8, group=Snowflake_G, parameterName=Flake_G_TileU}
  n23 = ScalarParameter() {defaultValue=6, group=Snowflake_G, parameterName=Flake_G_TileV}
  n24 = AppendVector(A=@n22, B=@n23)
  n25 = TextureSampleParameter2D(UVs=Frac(None=Multiply(A=@n21, B=@n24))) {group=SnowflakeSettings, parameterName=PanningTexture, texture=&T_14Days_Snowflakes}
  n26 = ScalarParameter() {defaultValue=32, group=SnowflakeSettings, parameterName=SnowflakeSharpness}
  n27 = AppendVector(A=Multiply(A=@n22) {constB=0.5}, B=Multiply(A=@n23) {constB=0.5})
  n28 = Add(A=Frac(None=Multiply(A=@n21, B=@n27))) {constB=-0.5}
  n29 = Clamp(None=Multiply(A=ComponentMask(None=@n28) {g=true}) {constB=256})
  n30 = Clamp(None=Multiply(A=ComponentMask(None=@n28) {r=true}) {constB=256})
  n31 = LinearInterpolate(B=Clamp(None=Multiply(A=@n25.G, B=@n26) {constB=32}), Alpha=OneMinus(None=LinearInterpolate(A=@n29, Alpha=@n30)))
  n32 = AppendVector(
   A=ScalarParameter() {defaultValue=0.1, group=Snowflake_B, parameterName=Flake_B_OffsetX}
   B=ScalarParameter() {defaultValue=0.25, group=Snowflake_B, parameterName=Flake_B_OffsetY}
  )
  n33 = Panner(
   Coordinate=Rotator(
    Coordinate=Add(A=@n7."X 100%, Y 100%", B=@n32) {constB=0.255}
    Time=ScalarParameter() {defaultValue=0.1, group=Snowflake_B, parameterName=Flake_B_Rotation}
   )
   Speed=AppendVector(
    A=ScalarParameter() {defaultValue=-0.05, group=Snowflake_B, parameterName=Flake_B_SpeedX}
    B=ScalarParameter() {defaultValue=-0.1, group=Snowflake_B, parameterName=Flake_B_SpeedY}
   )
  ) {speedX=0.25}
  n34 = ScalarParameter() {defaultValue=8, group=Snowflake_B, parameterName=Flake_B_TileU}
  n35 = ScalarParameter() {defaultValue=6, group=Snowflake_B, parameterName=Flake_B_TileV}
  n36 = Frac(None=Multiply(A=@n33, B=AppendVector(A=@n34, B=@n35)))
  n37 = Clamp(None=Multiply(
    A=TextureSampleParameter2D(UVs=@n36) {group=SnowflakeSettings, parameterName=PanningTexture, texture=&T_14Days_Snowflakes}.B
    B=ScalarParameter() {defaultValue=32, group=SnowflakeSettings, parameterName=SnowflakeSharpness}
   ) {constB=32})
  n38 = AppendVector(A=Multiply(A=@n34) {constB=0.5}, B=Multiply(A=@n35) {constB=0.5})
  n39 = Add(A=Frac(None=Multiply(A=@n33, B=@n38))) {constB=-0.5}
  n40 = Multiply(A=ComponentMask(None=@n39) {g=true}) {constB=256}
  n41 = Multiply(A=ComponentMask(None=@n39) {r=true}) {constB=256}
  n42 = OneMinus(None=LinearInterpolate(A=Clamp(None=@n40), Alpha=Clamp(None=@n41)))
  n43 = Clamp(None=Add(A=Add(A=@n19, B=@n31), B=LinearInterpolate(B=@n37, Alpha=@n42)))
  n44 = VectorParameter() {group=Colors, parameterName=Background_RGBA1, defaultValue=<rgba 0.0723 0.0844 1 1>}
  n45 = VectorParameter() {group=Colors, parameterName=Background_RGBA2, defaultValue=<rgba 0.6654 0.9301 1 1>}
  n46 = MaterialFunctionCall(
   UVs=@n1
   CenterPosition=AppendVector(
    A=ScalarParameter() {defaultValue=0.5, group=GradientSettings, parameterName=RadialGrad_XPositon}
    B=ScalarParameter() {group=GradientSettings, parameterName=RadialGrad_YPositon}
   )
   Radius=ScalarParameter() {defaultValue=1, group=GradientSettings, parameterName=RadialGrad_Radius}
   Density=ScalarParameter() {defaultValue=1, group=GradientSettings, parameterName=RadialGrad_Density}
  ) {materialFunction=&RadialGradientExponential}
  n47 = Add(A=@n1, B=ScalarParameter() {group=GradientSettings, parameterName=BackgroundGradientOffset}) {constB=0.5}
  n48 = ScalarParameter() {defaultValue=1, group=GradientSettings, parameterName=BackgroundGradientScale}
  n49 = ScalarParameter() {defaultValue=40, group=GradientSettings, parameterName=BackgroundGradientRotation}
  n50 = ComponentMask(None=Rotator(Coordinate=Multiply(A=@n47, B=@n48) {constB=0.5}, Time=Multiply(A=@n49) {constB=0.0028}) {centerY=0, speed=6.2832}) {g=true}
  n51 = StaticSwitchParameter(
   True=LinearInterpolate(A=@n44.RGB, B=@n45.RGB, Alpha=@n46.RadialGradientExponential)
   False=LinearInterpolate(A=@n44.RGB, B=@n45.RGB, Alpha=Clamp(None=@n50))
  ) {defaultValue=true, group=Switches, parameterName=UseRadialBackgroundGradient}
  n52 = Multiply(A=@n51, B=@n5.RGB) {constB=0.25}
  n53 = ScalarParameter() {defaultValue=-30, group=Ribbons, parameterName=NotchLeftAngle}
  n54 = TextureCoordinate() {coordinateIndex=4}
  n55 = Add(
   A=Rotator(Coordinate=@n54, Time=Multiply(A=@n53) {constB=0.0028}) {centerX=1, speed=6.2832}
   B=Multiply(A=ScalarParameter() {defaultValue=0.5, group=Ribbons, parameterName=NotchLeftOffset}) {constB=-1}
  )
  n56 = ScalarParameter() {defaultValue=64, group=Ribbons, parameterName=NotchEdgeSharpness}
  n57 = Clamp(None=Multiply(A=ComponentMask(None=@n55) {r=true}, B=@n56))
  n58 = Rotator(Coordinate=@n54, Time=Multiply(A=ScalarParameter() {defaultValue=30, group=Ribbons, parameterName=NotchRightAngle}) {constB=0.0028}) {centerX=0, speed=6.2832}
  n59 = Add(A=ScalarParameter() {defaultValue=0.5, group=Ribbons, parameterName=NotchRightOffset}) {constB=-1}
  n60 = Multiply(A=ComponentMask(None=Add(A=@n58, B=@n59)) {r=true}, B=@n56)
  n61 = OneMinus(None=LinearInterpolate(A=OneMinus(None=@n57), Alpha=Clamp(None=@n60)))
  n62 = TextureCoordinate() {coordinateIndex=4}
  n63 = Rotator(Coordinate=@n62, Time=Multiply(A=ScalarParameter() {defaultValue=10, group=Ribbons, parameterName=TopAngle}) {constB=0.0028}) {centerY=0, speed=6.2832}
  n64 = Multiply(A=ScalarParameter() {defaultValue=0.125, group=Ribbons, parameterName=TopOffset}) {constB=-1}
  n65 = ScalarParameter() {defaultValue=8, group=Ribbons, parameterName=TopBottomEdgeSharpness}
  n66 = Multiply(A=ComponentMask(None=Add(A=@n63, B=@n64)) {g=true}, B=@n65)
  n67 = Rotator(Coordinate=@n62, Time=Multiply(A=ScalarParameter() {defaultValue=10, group=Ribbons, parameterName=BottomAngle}) {constB=0.0028}) {centerY=1, speed=6.2832}
  n68 = Add(A=ScalarParameter() {defaultValue=0.125, group=Ribbons, parameterName=BottomOffset}) {constB=-1}
  n69 = Multiply(A=ComponentMask(None=Add(A=@n67, B=@n68)) {g=true}, B=@n65)
  n70 = TextureCoordinate() {coordinateIndex=4}
  n71 = Rotator(Coordinate=@n70, Time=Multiply(A=ScalarParameter() {defaultValue=10, group=Ribbons, parameterName=LeftAngle}) {constB=0.0028}) {centerX=1, speed=6.2832}
  n72 = Multiply(A=ScalarParameter() {defaultValue=0.125, group=Ribbons, parameterName=LeftOffset}) {constB=-1}
  n73 = ScalarParameter() {defaultValue=8, group=Ribbons, parameterName=RightLeftEdgeSharpness}
  n74 = Multiply(A=ComponentMask(None=Add(A=@n71, B=@n72)) {r=true}, B=@n73)
  n75 = Rotator(Coordinate=@n70, Time=Multiply(A=ScalarParameter() {defaultValue=10, group=Ribbons, parameterName=RightAngle}) {constB=0.0028}) {centerX=0, speed=6.2832}
  n76 = Add(A=ScalarParameter() {defaultValue=0.125, group=Ribbons, parameterName=RightOffset}) {constB=-1}
  n77 = Multiply(A=ComponentMask(None=Add(A=@n75, B=@n76)) {r=true}, B=@n73)
  n78 = DotProduct(A=Clamp(None=DotProduct(A=@n66, B=@n69)), B=Clamp(None=DotProduct(A=@n74, B=@n77)))
  LinearInterpolate(
   A=StaticSwitchParameter(True=LinearInterpolate(A=@n4, B=@n6, Alpha=@n43), False=@n4) {defaultValue=true, group=Switches, parameterName=UseSnowFlakesOnRibbon}
   B=StaticSwitchParameter(True=LinearInterpolate(A=@n51, B=@n52, Alpha=@n43), False=@n51) {group=Switches, parameterName=UseSnowFlakesOnBackground}
   Alpha=StaticSwitchParameter(True=LinearInterpolate(A=@n61, Alpha=@n78), False=ScalarParameter() {group=Switches, parameterName=GradientType}) {defaultValue=true, group=Switches, parameterName=UseRibbonMask}
  )
  StaticSwitchParameter(True=LinearInterpolate(B=LinearInterpolate(A=@n44.A, B=@n45.A, Alpha=@n46.RadialGradientExponential), Alpha=@n43), False=Constant() {r=1}) {group=Switches, parameterName=UseSnowFlakeAlphaMask}
 ]
}
