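# Translucent UI material for ribbon art in the 14 Days frontend, built from a channel-packed ribbon texture.
# The IsCenterRibbon static switch selects either flat RGB_1 with the blue-channel shape mask, or an RGB_1 to RGB_2 blend by the green channel with the red-channel shape mask.
# Each mask is multiplied by EdgeSharpness and clamped to produce crisp edges.
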Material /Game/Athena/UI/Frontend/Art/M_UI_14DaysRibbons
graph {
 properties {
  bUsedWithStaticMesh=false
  blendMode=BLEND_Translucent
  materialDomain=MD_UI
  pixelDepthOffsetMode=PDOM_Legacy
  usedShadingModels=Default Lit
 }
 nodes [
  n1 = VectorParameter() {parameterName=RGB_1, defaultValue=<rgba 0.0823 0 0.7913 1>}
  n2 = TextureSample() {texture=/Game/UI/Foundation/Textures/Pattern/T_14Days_Ribbons}
  n3 = ScalarParameter() {defaultValue=128, parameterName=EdgeSharpness}
  n4 = Multiply(A=@n2.G, B=@n3) {constB=64}
  StaticSwitchParameter(
   True=@n1.RGB
   False=LinearInterpolate(A=@n1.RGB, B=VectorParameter() {parameterName=RGB_2, defaultValue=<rgba 0.0612 0 0.5711 1>}.RGB, Alpha=Clamp(None=@n4))
  ) {parameterName=IsCenterRibbon}
  StaticSwitchParameter(True=Clamp(None=Multiply(A=@n2.B, B=@n3) {constB=64}), False=Clamp(None=Multiply(A=@n2.R, B=@n3) {constB=64})) {parameterName=IsCenterRibbon}
 ]
}
