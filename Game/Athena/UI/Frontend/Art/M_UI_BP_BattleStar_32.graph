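# Translucent UI material for the Battle Pass battle star icon in the frontend, with dimming driven by availability and hover.
# It blends between a near-black overlay color and the RGB of a swappable, clamp-sampled star texture parameter.
# Nested lerps set the blend weight: Overlay Value is the dimmed level, and both Available and Hover push it toward the full texture color.
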Material /Game/Athena/UI/Frontend/Art/M_UI_BP_BattleStar_32
graph {
 properties {
  bUsedWithStaticMesh=false
  blendMode=BLEND_Translucent
  materialDomain=MD_UI
  pixelDepthOffsetMode=PDOM_Legacy
  usedShadingModels=Default Lit
 }
 nodes [
  n1 = ScalarParameter() {defaultValue=0.5, parameterName=Overlay Value, sliderMax=1}
  n2 = ScalarParameter() {defaultValue=1, parameterName=Available, sliderMax=1}
  LinearInterpolate(
   A=VectorParameter() {parameterName=Overlay_Color, defaultValue=<rgba 0.013 0.013 0.013 1>}.RGB
   B=TextureSample(Tex=TextureObjectParameter() {parameterName=Texture, texture=&/T_UI_BP_BattleStar_64}) {samplerSource=SSM_Clamp_WorldGroupSettings, texture=&T_UI_BP_BattleStar_64}.RGB
   Alpha=LinearInterpolate(A=LinearInterpolate(A=@n1, Alpha=@n2), Alpha=ScalarParameter() {parameterName=Hover, sliderMax=1})
  )
 ]
}
