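# Translucent unlit glow material for the frontend Black Monday sign, tinted by a brightness-scaled GlowColor.
# Static switches pick the mask shape: light cone (view-angle and vertical fade), flare texture, or inverted symbol texture within a sphere mask.
# Distance and depth fades shape that mask, which is then modulated by panning triplanar or UV fog.
# Particle alpha and two screen-aligned scrolling cloud samples further vary the result.
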Material /Game/Athena/UI/Frontend/Art/M_UI_BlackMondaySign
graph {
 properties {blendMode=BLEND_Translucent, pixelDepthOffsetMode=PDOM_Legacy, shadingModel=MSM_Unlit, usedShadingModels=Unlit}
 nodes [
  MaterialFunctionCall() {materialFunction=&Fresnel_Function}
  n1 = VectorParameter() {parameterName=GlowColor, defaultValue=<rgba 1 0.5328 0.124 1>}
  MaterialFunctionCall(
   Color_Morning=Multiply(A=ScalarParameter() {defaultValue=1, group=Color Adjustments, parameterName=Morning Multiplier}, B=@n1.RGB) {constB=250}
   Color_Day=Multiply(A=@n1.RGB, B=ScalarParameter() {defaultValue=1, group=Color Adjustments, parameterName=Day Multiplier}) {constB=250}
   Color_Evening=Multiply(A=ScalarParameter() {defaultValue=1, group=Color Adjustments, parameterName=Evening Multiplier}, B=@n1.RGB) {constB=250}
   Color_Night=Multiply(A=ScalarParameter() {defaultValue=1, group=Color Adjustments, parameterName=Night Multiplier}, B=@n1.RGB) {constB=250}
  ) {materialFunction=&Color_Time_of_Day}
  n2 = ScalarParameter() {parameterName=DepthFade}
  n3 = ScalarParameter() {defaultValue=4, group=Fresnel, parameterName=Base Fresnel Exponent}
  n4 = StaticSwitchParameter(
   True=MaterialFunctionCall("FadeSensitivity (optional)"=@n3) {materialFunction=&FadeBasedOnViewAngle}.Result
   False=OneMinus(None=Fresnel(ExponentIn=@n3))
  ) {parameterName=UseViewAngleFade}
  n5 = ScalarParameter() {defaultValue=12, group=Fresnel, parameterName=Base Fresnel Power}
  n6 = MaterialFunctionCall(
   Position=WorldPosition().XYZ
   Fade_Start_Distance=ScalarParameter() {defaultValue=250, group=Fade, parameterName="Fade_Start_Distance (S)"}
   Fade_End_Distance=ScalarParameter() {defaultValue=100, group=Fade, parameterName="Fade_End_Distance (S)"}
  ) {materialFunction=&MF_Distance_SyncedFade_FX}
  n7 = Power(Base=ComponentMask(None=TextureCoordinate()) {g=true}, Exp=ScalarParameter() {defaultValue=1, group=Fade, parameterName=Vertical Fade Power})
  n8 = ScalarParameter() {defaultValue=2, group=Fade, parameterName=Vertical Fade Mult}
  n9 = Multiply(A=Multiply(A=Power(Base=@n4, Exp=@n5), B=@n6.Result), B=Saturate(None=Multiply(A=@n7, B=@n8)))
  n10 = MaterialFunctionCall(Texture Scale=ScalarParameter() {defaultValue=1.2, parameterName=CutoutTextureScale}) {materialFunction=&ScaleUVsByCenter}
  n11 = TextureSampleParameter2D!(UVs=@n10.UVs, Bias, Apply View MipBias) {group=Flare, mipValueMode=TMVM_MipBias, parameterName=Flare Texture, texture=/Game/Effects/Fort_Effects/Textures/Flares/T_Flare_Soft}
  n12 = TextureSampleParameter2D(UVs=@n10.UVs, Bias=ScalarParameter() {defaultValue=-1, parameterName=Symbol LOD Bias}) {
   group=End Symbol
   mipValueMode=TMVM_MipBias
   parameterName=Symbol Texture
   samplerSource=SSM_Clamp_WorldGroupSettings
   texture=/Game/Global/Textures/Default/Blanks/T_Blank_White_ZeroAlpha_sRGBColor
  }
  n13 = Multiply(A=OneMinus(None=@n12.R)) {constB=0.5}
  n14 = Power(
   Base=SphereMask(
    A=TextureCoordinate() {coordinateIndex=1}
    B=Constant() {r=0.5}
    Radius=ScalarParameter() {defaultValue=0.5, group=End Symbol, parameterName=Radius}
    Hardness=ScalarParameter() {defaultValue=0.5, group=End Symbol, parameterName=Hardness}
   )
   Exp=ScalarParameter() {defaultValue=2, group=End Symbol, parameterName=SphereMask Power}
  )
  n15 = Multiply(
   A=@n6.Result
   B=StaticSwitchParameter(
    True=StaticComponentMaskParameter(None=@n11.RGB) {defaultG=true, group=Flare, parameterName=Flare Texture Channel}
    False=Multiply(A=@n13, B=@n14)
   ) {parameterName=IsFlare?}
  )
  n16 = Power(
   Base=Multiply(A=DepthFade(FadeDistance=@n2) {fadeDistanceDefault=25}, B=StaticSwitchParameter(True=@n9, False=@n15) {parameterName=IsLightCone?})
   Exp=ScalarParameter() {defaultValue=1, parameterName=Final Power}
  )
  n17 = AppendVector(
   A=ScalarParameter() {defaultValue=0.005, group=Fog, parameterName=Fog Pan Speed X}
   B=ScalarParameter() {group=Fog, parameterName=Fog Pan Speed Y}
  )
  n18 = VectorParameter() {group=Fog, parameterName=FogTextureTiling, defaultValue=<rgba 2 2 2 1>}
  n19 = MaterialFunctionCall!(
   CameraVector
   Time Scale
   Panner Speed=@n17
   Texture=TextureObjectParameter() {group=Fog, parameterName=Fog Texture, texture=&T_Smooth_tile}
   Tiling=@n18.RGB
   Offset
   AxisFadeContrast
  ) {materialFunction=/Game/Weapons/WeaponSkins/MaterialFunction/TriplanarCamerVector_Panning}
  n20 = TextureCoordinate()
  n21 = ComponentMask(None=@n18.RGB) {g=true, r=true}
  n22 = Multiply(A=Time(), B=@n17)
  n23 = TextureSample(UVs=Add(A=Multiply(A=@n20, B=@n21), B=Frac(None=@n22))) {texture=&T_Smooth_tile}
  n24 = Multiply(
   A=Multiply(A=Clamp(None=@n16), B=ScalarParameter() {defaultValue=0.1, parameterName=Final Multiplier})
   B=LinearInterpolate(
    A=StaticSwitchParameter(True=@n19.Result, False=@n23.G) {defaultValue=true, group=Fog, parameterName=UseTriPlanarFog?}
    Alpha=ScalarParameter() {defaultValue=0.5, group=Fog, parameterName=Fade Fog}
   )
  )
  n25 = Multiply(A=MaterialFunctionCall(Texture Dimensions=Constant2Vector() {g=256, r=256}) {materialFunction=&ScreenAlignedUVs}."X 100%, Y 100%") {constB=0.85}
  n26 = Panner(Coordinate=@n25) {speedX=0.07, speedY=0.02}
  n27 = Multiply(
   A=Multiply(A=TextureSample(UVs=@n26) {texture=/Game/Packages/Fortnite_UniqueMaterials_1/PostProcess/T_TilingClouds_02}.R)
   B=TextureSample(UVs=Panner(Coordinate=@n25) {speedX=0.05}) {texture=&T_TilingClouds_02}.R
  )
  Multiply(A=Multiply(A=ParticleColor().A, B=Clamp(None=@n24)), B=Clamp(None=Add(A=@n27) {constB=0.5}))
  Multiply(A=@n1.RGB, B=ScalarParameter() {defaultValue=1.2, parameterName=BirghtnessScalar})
  TextureCoordinate()
 ]
}
