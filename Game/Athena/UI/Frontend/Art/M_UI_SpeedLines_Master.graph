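# Translucent UI master material that procedurally draws animated radial speed lines, without textures, for frontend screens.
# Time-rotated UVs go through arctangent and frac to form angular streaks, offset by radial distance and PullBack, then sharpened with an Edge/EdgeSoftness SmoothStep.
# WhiteScreenFadeOut blends from solid white to the lines.
# A LinearSine pulse feeds the pull-back, but both its lerp ends read SineExtents red, leaving it constant.
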Material /Game/Athena/UI/Frontend/Art/M_UI_SpeedLines_Master
graph {
 properties {
  bUsedWithStaticMesh=false
  blendMode=BLEND_Translucent
  materialDomain=MD_UI
  numCustomizedUVs=1
  pixelDepthOffsetMode=PDOM_Legacy
  usedShadingModels=Default Lit
 }
 nodes [
  n1 = Add(A=Multiply(A=Time(), B=ScalarParameter() {parameterName=RotationSpeed}), B=ScalarParameter() {defaultValue=0.3, parameterName=Rotation})
  n2 = ScalarParameter() {parameterName=RotationBoost}
  n3 = Subtract(A=Rotator(Coordinate=TextureCoordinate(), Time=Add(A=@n1, B=@n2))) {constB=0.5}
  n4 = ComponentMask(None=@n3) {g=true}
  n5 = ComponentMask(None=@n3) {r=true}
  n6 = Multiply(A=Divide(A=@n4, B=@n5))
  Abs(None=@n6)
  n7 = Multiply(A=Arctangent!(None=@n6)) {constB=1.8}
  n8 = Subtract(A=TextureCoordinate()) {constB=0.5}
  n9 = ComponentMask(None=@n8) {g=true}
  n10 = ComponentMask(None=@n8) {r=true}
  n11 = Add(A=Power(Base=Abs(None=@n9)), B=Power(Base=Abs(None=@n10)))
  n12 = Multiply(A=OneMinus(None=SquareRoot(None=@n11))) {constB=0.9}
  n13 = VectorParameter() {parameterName=SineExtents, defaultValue=<rgba 0 0 0 1>}
  n14 = Time()
  n15 = ScalarParameter() {parameterName=SineSpeed}
  n16 = MaterialFunctionCall(Value=Fmod(A=Multiply(A=@n14, B=@n15), B=Constant() {r=1})) {materialFunction=&LinearSine}
  n17 = LinearInterpolate(Alpha=Add(A=LinearInterpolate(A=@n13.R, B=@n13.R, Alpha=@n16.Linear Sine), B=ScalarParameter() {parameterName=PullBack})) {constB=0.3}
  n18 = Clamp(None=Subtract(A=Frac(None=@n7), B=Add(A=@n12, B=@n17) {constB=0}) {constB=0.95})
  Clamp(None=Multiply(A=Subtract(A=@n18) {constB=0.5}) {constB=100})
  Constant() {r=1}
  n19 = ScalarParameter() {defaultValue=0.495, parameterName=Edge}
  n20 = Multiply(A=@n19, B=ScalarParameter() {defaultValue=0.02, parameterName=EdgeSoftness})
  LinearInterpolate(
   A=Constant() {r=1}
   B=MaterialFunctionCall(Alpha=@n18, Min=@n19, Max=Add(A=@n19, B=@n20)) {materialFunction=&SmoothStep}.Result
   Alpha=ScalarParameter() {parameterName=WhiteScreenFadeOut}
  )
  MaterialFunctionCall() {materialFunction=&DebugTimeSine}
  n21 = ScalarParameter() {defaultValue=0.5}
  Subtract(A=Abs(None=@n4), B=@n21)
  Subtract(A=Abs(None=@n5), B=@n21)
 ]
}
